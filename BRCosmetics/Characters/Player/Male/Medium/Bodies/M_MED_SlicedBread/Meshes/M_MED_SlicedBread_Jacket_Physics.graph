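# Physics asset for the SlicedBread male medium player body's jacket, setting up secondary motion for its dynamic jacket bones.
# Ten two-segment capsule chains around front, sides and back hang from tiny MSTR anchor bodies, each constrained to the pelvis with free rotation.
# Swing and twist limits vary by chain, tightest on the bk_cr pair.
# Pelvis, spine and thigh capsules model the body volume the jacket can collide against.
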PhysicsAsset /BRCosmetics/Characters/Player/Male/Medium/Bodies/M_MED_SlicedBread/Meshes/M_MED_SlicedBread_Jacket_Physics
bodies {
 dyn_jacket_bk_cr_l_01 [{shapeType=Capsule, length=13.7042, radius=9.1084, shapeName=dyn_jacket_bk_01_capsule, center=<6.4931 0 7>, rotation=<rot 0 0 90>}]
 dyn_jacket_bk_cr_l_01_MSTR [{shapeType=Capsule, length=0.505, radius=0.505, shapeName=dyn_jacket_bk_cr_l_01_MSTR_capsule, rotation=<rot 0 -90 0>}]
 dyn_jacket_bk_cr_l_02 [{shapeType=Capsule, length=4.2423, radius=8.7895, shapeName=dyn_jacket_bk_02_capsule, center=<8.003 0 7>, rotation=<rot 0 0 90>}]
 dyn_jacket_bk_cr_r_01 [{shapeType=Capsule, length=13.7042, radius=9.1649, shapeName=dyn_jacket_bk_01_capsule, center=<6.4931 0 7>, rotation=<rot 0 0 90>}]
 dyn_jacket_bk_cr_r_01_MSTR [{shapeType=Capsule, length=0.505, radius=0.505, shapeName=dyn_jacket_bk_cr_r_01_MSTR_capsule, rotation=<rot 0 -90 0>}]
 dyn_jacket_bk_cr_r_02 [{shapeType=Capsule, length=4.2423, radius=8.8461, shapeName=dyn_jacket_bk_02_capsule, center=<8.003 0 7>, rotation=<rot 0 0 90>}]
 dyn_jacket_bk_l_01 [{shapeType=Capsule, length=14.8175, radius=8.0047, shapeName=dyn_jacket_bk_l_02_capsule, center=<8.2143 0 7>, rotation=<rot 0 0 90>}]
 dyn_jacket_bk_l_01_MSTR [{shapeType=Capsule, length=0.505, radius=0.505, shapeName=dyn_jacket_bk_l_01_MSTR_capsule, rotation=<rot 0 -90 0>}]
 dyn_jacket_bk_l_02 [{shapeType=Capsule, length=15.7541, radius=7.0681, shapeName=dyn_jacket_bk_l_02_capsule, center=<8.2143 0 7>, rotation=<rot 0 0 90>}]
 dyn_jacket_bk_r_01 [{shapeType=Capsule, length=15.8357, radius=8.4902, shapeName=dyn_jacket_bk_r_02_capsule, center=<8.3364 0 7>, rotation=<rot 0 0 90>}]
 dyn_jacket_bk_r_01_MSTR [{shapeType=Capsule, length=0.505, radius=0.505, shapeName=dyn_jacket_bk_r_01_MSTR_capsule, rotation=<rot 0 -90 0>}]
 dyn_jacket_bk_r_02 [{shapeType=Capsule, length=16.7724, radius=7.5535, shapeName=dyn_jacket_bk_r_02_capsule, center=<8.3364 0 7>, rotation=<rot 0 0 90>}]
 dyn_jacket_bkout_l_01 [{shapeType=Capsule, length=18.1102, radius=8.3312, shapeName=dyn_jacket_bkout_l_01_capsule, center=<10.2143 0 6.5>, rotation=<rot 0 0 90>}]
 dyn_jacket_bkout_l_01_MSTR [{shapeType=Capsule, length=0.505, radius=0.505, shapeName=dyn_jacket_bkout_l_01_MSTR_capsule, rotation=<rot 0 -90 0>}]
 dyn_jacket_bkout_l_02 [{shapeType=Capsule, length=14.6922, radius=8.3312, shapeName=dyn_jacket_bkout_l_02_capsule, center=<8.2143 0 6.5>, rotation=<rot 0 0 90>}]
 dyn_jacket_bkout_r_01 [{shapeType=Capsule, length=19.4776, radius=9.1947, shapeName=dyn_jacket_bkout_r_01_capsule, center=<8.0492 0 8>, rotation=<rot 0 0 90>}]
 dyn_jacket_bkout_r_01_MSTR [{shapeType=Capsule, length=0.505, radius=0.505, shapeName=dyn_jacket_bkout_r_01_MSTR_capsule, rotation=<rot 0 -90 0>}]
 dyn_jacket_bkout_r_02 [{shapeType=Capsule, length=19.4776, radius=9.1947, shapeName=dyn_jacket_bkout_r_02_capsule, center=<8.8045 0 8>, rotation=<rot 0 0 90>}]
 dyn_jacket_fr_l_01 [{shapeType=Capsule, length=7.4444, radius=8.5556, shapeName=dyn_jacket_fr_l_01_capsule, center=<6 -4 4.5>, rotation=<rot 90 0 0>}]
 dyn_jacket_fr_l_01_MSTR [{shapeType=Capsule, length=0.505, radius=0.505, shapeName=dyn_jacket_fr_l_01_MSTR_capsule, rotation=<rot 0 -90 0>}]
 dyn_jacket_fr_l_02 [{shapeType=Capsule, length=7.061, radius=7.0165, shapeName=dyn_jacket_fr_l_02_capsule, center=<2.9963 -4 4.5>, rotation=<rot 90 0 0>}]
 dyn_jacket_fr_r_01 [{shapeType=Capsule, length=8.7495, radius=10.0134, shapeName=dyn_jacket_fr_r_01_capsule, center=<6.3824 0 9>, rotation=<rot 0 0 90>}]
 dyn_jacket_fr_r_01_MSTR [{shapeType=Capsule, length=0.505, radius=0.505, shapeName=dyn_jacket_fr_r_01_MSTR_capsule, rotation=<rot 0 -90 0>}]
 dyn_jacket_fr_r_02 [{shapeType=Capsule, length=6.5802, radius=8.7361, shapeName=dyn_jacket_fr_r_01_capsule, center=<6.3824 0 9>, rotation=<rot 0 0 90>}]
 dyn_jacket_l_01 [{shapeType=Capsule, length=10, radius=6, shapeName=dyn_jacket_l_01_capsule, center=<6 0 6>, rotation=<rot 90 0 0>}]
 dyn_jacket_l_01_MSTR [{shapeType=Capsule, length=0.505, radius=0.505, shapeName=dyn_jacket_l_01_MSTR_capsule, rotation=<rot 0 -90 0>}]
 dyn_jacket_l_02 [{shapeType=Capsule, length=10, radius=6, shapeName=dyn_jacket_l_02_capsule, center=<4 0 6>, rotation=<rot 90 0 0>}]
 dyn_jacket_r_01 [{shapeType=Capsule, length=5.8061, radius=8.7361, shapeName=dyn_jacket_r_01_capsule, center=<7.9145 0 9>, rotation=<rot 0 0 90>}]
 dyn_jacket_r_01_MSTR [{shapeType=Capsule, length=0.505, radius=0.505, shapeName=dyn_jacket_r_01_MSTR_capsule, rotation=<rot 0 -90 0>}]
 dyn_jacket_r_02 [{shapeType=Capsule, length=5.6337, radius=4.5924, shapeName=dyn_jacket_r_02_capsule, center=<6.7546 0 5>, rotation=<rot 90 0 0>}]
 pelvis [
  {shapeType=Capsule, length=9.0958, radius=12.7277, center=<-9 0.16 0.25>}
  {shapeType=Capsule, length=14, radius=10, shapeName=pelvis_capsule, center=<0 -1.0702 0>}
 ]
 spine_01 [{shapeType=Capsule, length=8.3741, radius=9.6259, shapeName=spine_01_capsule, center=<0 -0.5 0>}]
 spine_02 [{shapeType=Capsule, length=8.6438, radius=9.3562, shapeName=spine_02_capsule, center=<0 -0.5 0>}]
 spine_03 [{shapeType=Capsule, length=10.4269, radius=11.0731, shapeName=spine_03_capsule}]
 thigh_l [
  {shapeType=Capsule, length=24.5, radius=10.5, shapeName=thigh_l_capsule, center=<-20 0 0>, rotation=<rot 90 0 0>}
  {shapeType=Capsule, length=24.5, radius=10.5, shapeName=thigh_l_capsule, center=<-20.0389 -1.2332 0>, rotation=<rot 90 180 -177.8499>}
  {
   shapeType=Capsule
   length=15.1038
   radius=7.1444
   shapeName=thigh_l_capsule
   center=<-16.7137 7.8154 8.9441>
   rotation=<rot 87.0182 0.0051 0.0093>
  }
 ]
 thigh_r [
  {shapeType=Capsule, length=23.6787, radius=11.8213, shapeName=thigh_r_capsule, center=<20 0 0>, rotation=<rot 90 0 0>}
  {shapeType=Capsule, length=24.5, radius=11, shapeName=thigh_r_capsule, center=<19.8379 2.0491 0>, rotation=<rot 90 89.6476 90.9378>}
 ]
}
constraints [
 {bone1Name=dyn_jacket_bk_cr_l_01, bone2Name=dyn_jacket_bk_cr_l_01_MSTR, swing1LimitDegrees=20, swing2LimitDegrees=45, twistLimitDegrees=5}
 {bone1Name=dyn_jacket_bk_cr_l_02, bone2Name=dyn_jacket_bk_cr_l_01, swing1LimitDegrees=20, swing2LimitDegrees=45, twistLimitDegrees=5}
 {bone1Name=dyn_jacket_bk_cr_r_01, bone2Name=dyn_jacket_bk_cr_r_01_MSTR, swing1LimitDegrees=20, swing2LimitDegrees=45, twistLimitDegrees=5}
 {bone1Name=dyn_jacket_bk_cr_r_02, bone2Name=dyn_jacket_bk_cr_r_01, swing1LimitDegrees=20, swing2LimitDegrees=45, twistLimitDegrees=5}
 {bone1Name=dyn_jacket_bk_l_01, bone2Name=dyn_jacket_bk_l_01_MSTR, swing1LimitDegrees=35, swing2LimitDegrees=60, twistLimitDegrees=25}
 {bone1Name=dyn_jacket_bk_l_02, bone2Name=dyn_jacket_bk_l_01, swing1LimitDegrees=35, swing2LimitDegrees=60, twistLimitDegrees=25}
 {bone1Name=dyn_jacket_bk_r_01, bone2Name=dyn_jacket_bk_r_01_MSTR, swing1LimitDegrees=35, swing2LimitDegrees=60, twistLimitDegrees=25}
 {bone1Name=dyn_jacket_bk_r_02, bone2Name=dyn_jacket_bk_r_01, swing1LimitDegrees=35, swing2LimitDegrees=60, twistLimitDegrees=25}
 {bone1Name=dyn_jacket_bkout_l_01, bone2Name=dyn_jacket_bkout_l_01_MSTR, swing1LimitDegrees=35, swing2LimitDegrees=60, twistLimitDegrees=25}
 {bone1Name=dyn_jacket_bkout_l_02, bone2Name=dyn_jacket_bkout_l_01, swing1LimitDegrees=35, swing2LimitDegrees=60, twistLimitDegrees=25}
 {bone1Name=dyn_jacket_bkout_r_01, bone2Name=dyn_jacket_bkout_r_01_MSTR, swing1LimitDegrees=35, swing2LimitDegrees=60, twistLimitDegrees=25}
 {bone1Name=dyn_jacket_bkout_r_02, bone2Name=dyn_jacket_bkout_r_01, swing1LimitDegrees=35, swing2LimitDegrees=60, twistLimitDegrees=25}
 {bone1Name=dyn_jacket_fr_l_01, bone2Name=dyn_jacket_fr_l_01_MSTR, swing1LimitDegrees=60, swing2LimitDegrees=60, twistLimitDegrees=60}
 {bone1Name=dyn_jacket_fr_l_02, bone2Name=dyn_jacket_fr_l_01, swing1LimitDegrees=60, swing2LimitDegrees=60, twistLimitDegrees=60}
 {bone1Name=dyn_jacket_fr_r_01, bone2Name=dyn_jacket_fr_r_01_MSTR, swing1LimitDegrees=45, swing2LimitDegrees=45, twistLimitDegrees=45}
 {bone1Name=dyn_jacket_fr_r_02, bone2Name=dyn_jacket_fr_r_01, swing1LimitDegrees=45, swing2LimitDegrees=45, twistLimitDegrees=45}
 {bone1Name=dyn_jacket_l_02, bone2Name=dyn_jacket_l_01, swing1LimitDegrees=35, swing2LimitDegrees=60, twistLimitDegrees=25}
 {bone1Name=dyn_jacket_l_01, bone2Name=dyn_jacket_l_01_MSTR, swing1LimitDegrees=35, swing2LimitDegrees=60, twistLimitDegrees=25}
 {bone1Name=dyn_jacket_r_01, bone2Name=dyn_jacket_r_01_MSTR, swing1LimitDegrees=45, swing2LimitDegrees=45, twistLimitDegrees=45}
 {bone1Name=dyn_jacket_r_02, bone2Name=dyn_jacket_r_01, swing1LimitDegrees=45, swing2LimitDegrees=45, twistLimitDegrees=45}
 {
  bone1Name=dyn_jacket_fr_l_01_MSTR
  bone2Name=pelvis
  swing1LimitDegrees=45
  swing1Motion=Free
  swing2LimitDegrees=45
  swing2Motion=Free
  twistLimitDegrees=45
  twistMotion=Free
 }
 {
  bone1Name=dyn_jacket_l_01_MSTR
  bone2Name=pelvis
  swing1LimitDegrees=45
  swing1Motion=Free
  swing2LimitDegrees=45
  swing2Motion=Free
  twistLimitDegrees=45
  twistMotion=Free
 }
 {
  bone1Name=dyn_jacket_bkout_l_01_MSTR
  bone2Name=pelvis
  swing1LimitDegrees=45
  swing1Motion=Free
  swing2LimitDegrees=45
  swing2Motion=Free
  twistLimitDegrees=45
  twistMotion=Free
 }
 {
  bone1Name=dyn_jacket_bk_l_01_MSTR
  bone2Name=pelvis
  swing1LimitDegrees=45
  swing1Motion=Free
  swing2LimitDegrees=45
  swing2Motion=Free
  twistLimitDegrees=45
  twistMotion=Free
 }
 {
  bone1Name=dyn_jacket_bk_cr_l_01_MSTR
  bone2Name=pelvis
  swing1LimitDegrees=45
  swing1Motion=Free
  swing2LimitDegrees=45
  swing2Motion=Free
  twistLimitDegrees=45
  twistMotion=Free
 }
 {
  bone1Name=dyn_jacket_bk_cr_r_01_MSTR
  bone2Name=pelvis
  swing1LimitDegrees=45
  swing1Motion=Free
  swing2LimitDegrees=45
  swing2Motion=Free
  twistLimitDegrees=45
  twistMotion=Free
 }
 {
  bone1Name=dyn_jacket_bk_r_01_MSTR
  bone2Name=pelvis
  swing1LimitDegrees=45
  swing1Motion=Free
  swing2LimitDegrees=45
  swing2Motion=Free
  twistLimitDegrees=45
  twistMotion=Free
 }
 {
  bone1Name=dyn_jacket_bkout_r_01_MSTR
  bone2Name=pelvis
  swing1LimitDegrees=45
  swing1Motion=Free
  swing2LimitDegrees=45
  swing2Motion=Free
  twistLimitDegrees=45
  twistMotion=Free
 }
 {
  bone1Name=dyn_jacket_r_01_MSTR
  bone2Name=pelvis
  swing1LimitDegrees=45
  swing1Motion=Free
  swing2LimitDegrees=45
  swing2Motion=Free
  twistLimitDegrees=45
  twistMotion=Free
 }
 {
  bone1Name=dyn_jacket_fr_r_01_MSTR
  bone2Name=pelvis
  swing1LimitDegrees=45
  swing1Motion=Free
  swing2LimitDegrees=45
  swing2Motion=Free
  twistLimitDegrees=45
  twistMotion=Free
 }
 {
  bone1Name=spine_01
  bone2Name=pelvis
  swing1LimitDegrees=45
  swing1Motion=Free
  swing2LimitDegrees=45
  swing2Motion=Free
  twistLimitDegrees=45
  twistMotion=Free
 }
 {
  bone1Name=spine_02
  bone2Name=spine_01
  swing1LimitDegrees=45
  swing1Motion=Free
  swing2LimitDegrees=45
  swing2Motion=Free
  twistLimitDegrees=45
  twistMotion=Free
 }
 {
  bone1Name=thigh_l
  bone2Name=pelvis
  swing1LimitDegrees=45
  swing1Motion=Free
  swing2LimitDegrees=45
  swing2Motion=Free
  twistLimitDegrees=45
  twistMotion=Free
 }
 {
  bone1Name=thigh_r
  bone2Name=pelvis
  swing1LimitDegrees=45
  swing1Motion=Free
  swing2LimitDegrees=45
  swing2Motion=Free
  twistLimitDegrees=45
  twistMotion=Free
 }
 {
  bone1Name=spine_03
  bone2Name=spine_02
  swing1LimitDegrees=45
  swing1Motion=Free
  swing2LimitDegrees=45
  swing2Motion=Free
  twistLimitDegrees=45
  twistMotion=Free
 }
]
