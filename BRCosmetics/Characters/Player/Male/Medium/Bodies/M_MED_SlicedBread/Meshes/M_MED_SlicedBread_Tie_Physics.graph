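# Physics asset for the SlicedBread male medium character's tie, giving the tie bones physical bodies and joints.
# Capsules on the pelvis, spine and thighs approximate the torso and upper legs, linked by free-rotating constraints.
# The tie chain hangs from spine_05 through a small master knot body, two knot bodies and three tie segments, with tight angular limits that widen toward the tip.
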PhysicsAsset /BRCosmetics/Characters/Player/Male/Medium/Bodies/M_MED_SlicedBread/Meshes/M_MED_SlicedBread_Tie_Physics
bodies {
 pelvis [
  {shapeType=Capsule, length=9, radius=8.5, shapeName=pelvis_capsule, center=<7.2135 0 -0.0001>}
  {shapeType=Capsule, length=19.5002, radius=18.7242, shapeName=pelvis_capsule, center=<-9.6979 -5.7741 -0.0001>}
 ]
 spine_01 [{shapeType=Capsule, length=6.7188, radius=10.9688, center=<-2.0733 -1.3709 0>}]
 spine_02 [{shapeType=Capsule, length=19.1562, radius=12.0938, center=<0 -0.6181 0>}]
 spine_03 [{shapeType=Capsule, length=19.1562, radius=12.0938, center=<0 -0.2866 0>}]
 spine_04 [{shapeType=Capsule, length=19.1562, radius=12.0938, center=<0 -0.1776 0>}]
 spine_05 [{shapeType=Capsule, length=20, radius=13.2188, center=<-4.5203 -1.893 0>, rotation=<rot 0 -0.0001 0>}]
 thigh_l [{shapeType=Capsule, length=26.0625, radius=11.1049, shapeName=thigh_r_capsule, center=<-21.6782 0 0>, rotation=<rot 90 0 0>}]
 thigh_r [{shapeType=Capsule, length=26.0625, radius=11.1049, shapeName=thigh_r_capsule, center=<21.9853 0 0>, rotation=<rot 90 0 0>}]
 tie_01_dyn [{shapeType=Capsule, length=8.6424, radius=1.9305, shapeName=dyn_tie_02_capsule, center=<6.7498 0 0>, rotation=<rot 90 -90 -90>}]
 tie_02_dyn [{shapeType=Capsule, length=7.0695, radius=1.9305, shapeName=dyn_tie_03_capsule, center=<4.805 0 0>, rotation=<rot 90 0.0099 0.0099>}]
 tie_03_dyn [{shapeType=Capsule, length=7.0695, radius=1.9305, shapeName=dyn_tie_04_capsule, center=<4.805 0 0>, rotation=<rot 90 0 0>}]
 tie_knot_01_dyn [{shapeType=Capsule, length=1.988, radius=3.2776, shapeName=dyn_tie_02_capsule, center=<3.3091 0 1.5079>, rotation=<rot 81.8591 180 -180>}]
 tie_knot_02_dyn [{shapeType=Capsule, length=4.2566, radius=3.342, shapeName=dyn_tie_02_capsule, center=<4.805 0 1.5079>, rotation=<rot 90 0.0099 0.0099>}]
 tie_knot_mstr [{shapeType=Capsule, length=0.505, radius=0.505, shapeName=tie_knot_mstr_capsule, rotation=<rot 0 -90 0>}]
}
constraints [
 {
  bone1Name=spine_01
  bone2Name=pelvis
  swing1LimitDegrees=45
  swing1Motion=Free
  swing2LimitDegrees=45
  swing2Motion=Free
  twistLimitDegrees=45
  twistMotion=Free
 }
 {
  bone1Name=spine_02
  bone2Name=spine_01
  swing1LimitDegrees=45
  swing1Motion=Free
  swing2LimitDegrees=45
  swing2Motion=Free
  twistLimitDegrees=45
  twistMotion=Free
 }
 {
  bone1Name=spine_03
  bone2Name=spine_02
  swing1LimitDegrees=45
  swing1Motion=Free
  swing2LimitDegrees=45
  swing2Motion=Free
  twistLimitDegrees=45
  twistMotion=Free
 }
 {
  bone1Name=spine_04
  bone2Name=spine_03
  swing1LimitDegrees=45
  swing1Motion=Free
  swing2LimitDegrees=45
  swing2Motion=Free
  twistLimitDegrees=45
  twistMotion=Free
 }
 {
  bone1Name=spine_05
  bone2Name=spine_04
  swing1LimitDegrees=45
  swing1Motion=Free
  swing2LimitDegrees=45
  swing2Motion=Free
  twistLimitDegrees=45
  twistMotion=Free
 }
 {bone1Name=tie_knot_01_dyn, bone2Name=tie_knot_mstr, swing1LimitDegrees=15, swing2LimitDegrees=15, twistLimitDegrees=5}
 {bone1Name=tie_01_dyn, bone2Name=tie_knot_02_dyn, swing1LimitDegrees=15, swing2LimitDegrees=15, twistLimitDegrees=5}
 {bone1Name=tie_02_dyn, bone2Name=tie_01_dyn, swing1LimitDegrees=18, swing2LimitDegrees=18, twistLimitDegrees=5}
 {bone1Name=tie_03_dyn, bone2Name=tie_02_dyn, swing1LimitDegrees=20, swing2LimitDegrees=20, twistLimitDegrees=20}
 {
  bone1Name=tie_knot_mstr
  bone2Name=spine_05
  swing1LimitDegrees=45
  swing1Motion=Free
  swing2LimitDegrees=45
  swing2Motion=Free
  twistLimitDegrees=45
  twistMotion=Free
 }
 {bone1Name=tie_knot_02_dyn, bone2Name=tie_knot_01_dyn, swing1LimitDegrees=15, swing2LimitDegrees=15, twistLimitDegrees=5}
 {
  bone1Name=thigh_r
  bone2Name=pelvis
  swing1LimitDegrees=45
  swing1Motion=Free
  swing2LimitDegrees=45
  swing2Motion=Free
  twistLimitDegrees=45
  twistMotion=Free
 }
 {
  bone1Name=thigh_l
  bone2Name=pelvis
  swing1LimitDegrees=45
  swing1Motion=Free
  swing2LimitDegrees=45
  swing2Motion=Free
  twistLimitDegrees=45
  twistMotion=Free
 }
]
