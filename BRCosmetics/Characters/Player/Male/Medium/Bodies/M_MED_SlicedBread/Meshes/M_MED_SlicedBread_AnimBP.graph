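# Animation Blueprint for the SlicedBread player body cosmetic, adding secondary motion to its jacket, tie and buttons.
# The parent mesh pose passes through a Control Rig into a cache feeding RigidBody jacket (normal or skydive) and tie simulations.
# Without rigid bodies, weighted constraints and crouch-dependent bone corrections substitute, before shared jacket offsets and button constraints.
# The EventGraph derives state flags and maps thigh rotations to corrective alphas; Trail controllers sit unconnected to the output.
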AnimBlueprint /BRCosmetics/Characters/Player/Male/Medium/Bodies/M_MED_SlicedBread/Meshes/M_MED_SlicedBread_AnimBP
target_skeleton=&/M_MED_SlicedBread_Skeleton
graphs [
 AnimGraph {nodes [
   "Constraint - tie_knot_mstr (P)"(Alpha=1, ConstraintWeights_0=1) {
    alphaBoolBlend {blendOption=Linear}
    boneToModify {boneName=tie_knot_mstr}
    constraintSetup [{targetBone {boneName=spine_05}}]
    constraintWeights [1]
   }
   Look At - Bone: tie_knot_mstr(Alpha=1) {
    alphaBoolBlend {blendOption=Linear}
    boneToModify {boneName=tie_knot_mstr}
    lookAtLocation=<-12.9434 40.4475 3.0517>
    lookAtTarget {boneReference {boneName=pelvis}}
    lookAt_Axis {axis=<1 0>}
   }
   n1 = RigidBody!(ComponentPose=Local To Component(LocalPose=Use cached pose 'PreSim'()), Alpha, OverrideWorldGravity:Vector="0,0,-980", ExternalForce) -> (Pose) {
    bOverrideWorldGravity=true
    overridePhysicsAsset=&/M_MED_SlicedBread_Jacket_Skydive_Physics
    simulationSpace=BaseBoneSpace
    alphaBoolBlend {blendOption=Linear}
    baseBoneRef {boneName=pelvis}
    componentAppliedLinearAccClamp=<1400 1400 1400>
    componentLinearAccScale=<0.4 0.3 0.1>
    componentLinearVelScale=<0.4 0.3 0.1>
    overrideWorldGravity {z=-980}
    simSpaceSettings {maxAngularAcceleration=100, maxAngularVelocity=1, maxLinearAcceleration=250, maxLinearVelocity=250, velocityScaleZ=0.3, worldAlpha=0.5}
   }
   n2 = Component To Local(ComponentPose=@n1)
   n3 = RigidBody(ComponentPose=Local To Component(LocalPose=Use cached pose 'PreSim'()), OverrideWorldGravity="0,0,-980") {
    bOverrideWorldGravity=true
    overridePhysicsAsset=&/M_MED_SlicedBread_Jacket_Physics
    simulationSpace=BaseBoneSpace
    alphaBoolBlend {blendOption=Linear}
    baseBoneRef {boneName=pelvis}
    componentAppliedLinearAccClamp=<1400 1400 1400>
    componentLinearAccScale=<0.4 0.3 0.1>
    componentLinearVelScale=<0.4 0.3 0.1>
    overrideWorldGravity {z=-980}
    simSpaceSettings {maxAngularAcceleration=100, maxAngularVelocity=1, maxLinearAcceleration=250, maxLinearVelocity=250, velocityScaleZ=0.3, worldAlpha=0.5}
   }
   n4 = Component To Local(ComponentPose=@n3)
   n5 = Get bIsSkyDiving()
   n6 = RigidBody(
    ComponentPose=Local To Component(LocalPose=Blend Poses by bool(BlendPose_0=@n2, BlendPose_1=@n4, bActiveValue=@n5) {blendType=Linear})
    OverrideWorldGravity="0,0,-980"
   ) {
    bClampLinearTranslationLimitToRefPose=true
    bOverrideWorldGravity=true
    overridePhysicsAsset=&/M_MED_SlicedBread_Tie_Physics
    simulationSpace=BaseBoneSpace
    alphaBoolBlend {blendOption=Linear}
    baseBoneRef {boneName=spine_03}
    componentAppliedLinearAccClamp=<3000 3000 3000>
    componentLinearAccScale=<0.4 0.4 0.5>
    overrideWorldGravity {z=-980}
    simSpaceSettings {
     maxAngularAcceleration=1000
     maxAngularVelocity=0.3
     maxLinearAcceleration=700
     maxLinearVelocity=500
     velocityScaleZ=0
     worldAlpha=0.2
     externalLinearDragV=<0.5 0.5>
    }
   }
   n7 = Component To Local(ComponentPose=@n6)
   n8 = "Constraint - tie_knot_mstr (P)"(ComponentPose=Local To Component(LocalPose=Use cached pose 'PreSim'()), Alpha=1, ConstraintWeights_0=1) {
    alphaBoolBlend {blendOption=Linear}
    boneToModify {boneName=tie_knot_mstr}
    constraintSetup [{targetBone {boneName=neck_01}}]
    constraintWeights [1]
   }
   n9 = Get thigh_rot_inr_r() -> (thigh_rot_inr_r:double)
   n10 = Look At - Bone: dyn_jacket_fr_r_01_MSTR(ComponentPose=@n8, Alpha=@n9) {
    alphaBoolBlend {blendOption=Linear}
    alphaScaleBiasClamp {bInterpResult=true}
    boneToModify {boneName=dyn_jacket_fr_r_01_MSTR}
    lookAtLocation=<0 -12 5>
    lookAtTarget {boneReference {boneName=calf_r}}
    lookAt_Axis {axis=<1 0>}
    lookUp_Axis {axis {x=0}}
   }
   n11 = Look At - Bone: dyn_jacket_r_01_MSTR(ComponentPose=@n10, Alpha=@n9) {
    alphaBoolBlend {blendOption=Linear}
    alphaScaleBiasClamp {bInterpResult=true}
    boneToModify {boneName=dyn_jacket_r_01_MSTR}
    lookAtLocation {x=0, z=11}
    lookAtTarget {boneReference {boneName=calf_r}}
    lookAt_Axis {axis=<1 0>}
    lookUp_Axis {axis {x=0}}
   }
   n12 = Get thigh_rot_up_r() -> (thigh_rot_up_r:double)
   n13 = "Transform (Modify) Bone - Bone: dyn_jacket_fr_r_01"(ComponentPose=@n11, Alpha=@n12, Translation="0,0,0", Rotation="9.096,4.987,10.829", Scale="1,1,1") {
    rotationMode=BMM_Additive
    rotationSpace=BCS_BoneSpace
    translationMode=BMM_Additive
    translationSpace=BCS_BoneSpace
    alphaBoolBlend {blendOption=Linear}
    boneToModify {boneName=dyn_jacket_fr_r_01}
    rotation=<rot 9.096 4.987 10.829>
   }
   n14 = "Transform (Modify) Bone - Bone: dyn_jacket_r_01"(ComponentPose=@n13, Alpha=@n12, Translation="0,0,0", Rotation="5.000,0.000,0.000", Scale="1,1,1") {
    rotationMode=BMM_Additive
    rotationSpace=BCS_BoneSpace
    translationMode=BMM_Additive
    translationSpace=BCS_BoneSpace
    alphaBoolBlend {blendOption=Linear}
    boneToModify {boneName=dyn_jacket_r_01}
    rotation {pitch=5}
   }
   n15 = Get thigh_rot_inr_l() -> (thigh_rot_inr_l:double)
   n16 = Look At - Bone: dyn_jacket_fr_l_01_MSTR(ComponentPose=@n14, Alpha=@n15) {
    alphaBoolBlend {blendOption=Linear}
    alphaScaleBiasClamp {bInterpResult=true}
    boneToModify {boneName=dyn_jacket_fr_l_01_MSTR}
    lookAtLocation=<0 12 -5>
    lookAtTarget {boneReference {boneName=calf_l}}
    lookAt_Axis {axis=<1 0>}
    lookUp_Axis {axis {x=0}}
   }
   n17 = Look At - Bone: dyn_jacket_l_01_MSTR(ComponentPose=@n16, Alpha=@n15) {
    alphaBoolBlend {blendOption=Linear}
    alphaScaleBiasClamp {bInterpResult=true}
    boneToModify {boneName=dyn_jacket_l_01_MSTR}
    lookAtLocation {x=0, z=-11}
    lookAtTarget {boneReference {boneName=calf_l}}
    lookAt_Axis {axis=<1 0>}
    lookUp_Axis {axis {x=0}}
   }
   n18 = Get thigh_rot_up_l() -> (thigh_rot_up_l:double)
   n19 = "Transform (Modify) Bone - Bone: dyn_jacket_fr_l_01"(ComponentPose=@n17, Alpha=@n18, Translation="0,0,0", Rotation="9.096,-4.987,-10.828", Scale="1,1,1") {
    rotationMode=BMM_Additive
    rotationSpace=BCS_BoneSpace
    translationMode=BMM_Additive
    translationSpace=BCS_BoneSpace
    alphaBoolBlend {blendOption=Linear}
    boneToModify {boneName=dyn_jacket_fr_l_01}
    rotation=<rot 9.096 -4.987 -10.8284>
   }
   n20 = "Transform (Modify) Bone - Bone: dyn_jacket_l_01"(ComponentPose=@n19, Alpha=@n18, Translation="0,0,0", Rotation="5,0,0", Scale="1,1,1") {
    rotationMode=BMM_Additive
    rotationSpace=BCS_BoneSpace
    translationMode=BMM_Additive
    translationSpace=BCS_BoneSpace
    alphaBoolBlend {blendOption=Linear}
    boneToModify {boneName=dyn_jacket_l_01}
    rotation {pitch=5}
   }
   n21 = Component To Local(ComponentPose=@n20)
   n22 = "Constraint - tie_knot_mstr (P)"(ComponentPose=Local To Component(LocalPose=Copy Pose From Mesh() {bUseAttachedParent=true}), Alpha=1, ConstraintWeights_0=1) {
    alphaBoolBlend {blendOption=Linear}
    boneToModify {boneName=tie_knot_mstr}
    constraintSetup [{targetBone {boneName=neck_02}}]
    constraintWeights [1]
   }
   n23 = "Constraint - dyn_jacket_fr_l_01_MSTR (PPPPP)"!(
    ComponentPose=@n22
    Alpha=1
    ConstraintWeights_0=0.0107
    ConstraintWeights_1:float=0.3164
    ConstraintWeights_2:float=0.0004
    ConstraintWeights_3:float=0.6567
    ConstraintWeights_4:float=0.0159
   ) -> (Pose) {
    alphaBoolBlend {blendOption=Linear}
    boneToModify {boneName=dyn_jacket_fr_l_01_MSTR}
    constraintSetup [
     {targetBone {boneName=pelvis}}
     {targetBone {boneName=thigh_l}}
     {targetBone {boneName=deform_knee_out_base_l}}
     {targetBone {boneName=thigh_twist_01_l}}
     {targetBone {boneName=deform_groin_l}}
    ]
    constraintWeights [1,1,1,1,1]
   }
   n24 = "Constraint - dyn_jacket_l_01_MSTR (P)"(ComponentPose=@n23, Alpha=0.3, ConstraintWeights_0=1) {
    alphaBoolBlend {blendOption=Linear}
    boneToModify {boneName=dyn_jacket_l_01_MSTR}
    constraintSetup [{targetBone {boneName=dyn_jacket_fr_l_01_MSTR}}]
    constraintWeights [1]
   }
   n25 = "Constraint - dyn_jacket_fr_r_01_MSTR (R)"(ComponentPose=@n24, Alpha=1, ConstraintWeights_0=1) {
    alphaBoolBlend {blendOption=Linear}
    boneToModify {boneName=dyn_jacket_fr_r_01_MSTR}
    constraintSetup [{transformType=Rotation, targetBone {boneName=thigh_r}}]
    constraintWeights [1]
   }
   n26 = "Transform (Modify) Bone - Bone: dyn_jacket_fr_l_01_MSTR"(ComponentPose=@n25, Alpha=1, Translation="28.841,-13.322,10.148", Rotation="-68.040,-78.774,14.844", Scale="1,1,1") {
    rotationMode=BMM_Additive
    rotationSpace=BCS_BoneSpace
    scaleMode=BMM_Additive
    scaleSpace=BCS_BoneSpace
    translationMode=BMM_Additive
    translationSpace=BCS_BoneSpace
    alphaBoolBlend {blendOption=Linear}
    boneToModify {boneName=dyn_jacket_fr_l_01_MSTR}
    rotation=<rot -68.04 -78.774 14.844>
    translation=<28.841 -13.322 10.148>
   }
   n27 = "Transform (Modify) Bone - Bone: dyn_jacket_l_01_MSTR"(ComponentPose=@n26, Alpha=Get thigh_rot_inr_l(), Translation="5.452,-2.770,2.132", Rotation="17.102,-27.184,7.574", Scale="1,1,1") {
    rotationMode=BMM_Additive
    rotationSpace=BCS_BoneSpace
    scaleMode=BMM_Additive
    scaleSpace=BCS_BoneSpace
    translationMode=BMM_Additive
    translationSpace=BCS_BoneSpace
    alphaBoolBlend {blendOption=Linear}
    boneToModify {boneName=dyn_jacket_l_01_MSTR}
    rotation=<rot 17.102 -27.184 7.574>
    translation=<5.452 -2.77 2.132>
   }
   n28 = "Transform (Modify) Bone - Bone: dyn_jacket_fr_r_01_MSTR"(ComponentPose=@n27, Alpha=1, Translation="9.881,3.003,5.851", Rotation="0.093,1.961,-0.084", Scale="1,1,1") {
    rotationMode=BMM_Additive
    rotationSpace=BCS_BoneSpace
    scaleMode=BMM_Additive
    scaleSpace=BCS_BoneSpace
    translationMode=BMM_Additive
    translationSpace=BCS_BoneSpace
    alphaBoolBlend {blendOption=Linear}
    boneToModify {boneName=dyn_jacket_fr_r_01_MSTR}
    rotation=<rot 0.093 1.961 -0.084>
    translation=<9.881 3.003 5.851>
   }
   n29 = "Transform (Modify) Bone - Bone: dyn_jacket_r_01_MSTR"(ComponentPose=@n28, Alpha=1, Translation="-2.293,2.670,-0.413", Rotation="14.145,28.648,6.089", Scale="1,1,1") {
    rotationMode=BMM_Additive
    rotationSpace=BCS_BoneSpace
    scaleMode=BMM_Additive
    scaleSpace=BCS_BoneSpace
    translationMode=BMM_Additive
    translationSpace=BCS_BoneSpace
    alphaBoolBlend {blendOption=Linear}
    boneToModify {boneName=dyn_jacket_r_01_MSTR}
    rotation=<rot 14.145 28.648 6.089>
    translation=<-2.293 2.67 -0.413>
   }
   n30 = "Transform (Modify) Bone - Bone: dyn_jacket_r_01_MSTR"(ComponentPose=@n29, Alpha=1, Translation="-0.000,1.711,1.382", Rotation="0.000,-0.000,35.084", Scale="1,1,1") {
    rotationMode=BMM_Additive
    rotationSpace=BCS_BoneSpace
    scaleMode=BMM_Additive
    scaleSpace=BCS_BoneSpace
    translationMode=BMM_Additive
    translationSpace=BCS_BoneSpace
    alphaBoolBlend {blendOption=Linear}
    boneToModify {boneName=dyn_jacket_r_01_MSTR}
    rotation=<rot 0 0 35.084>
    translation {y=1.711, z=1.382}
   }
   n31 = "Transform (Modify) Bone - Bone: tie_knot_mstr"(ComponentPose=@n30, Alpha=1, Translation="0,0,0", Rotation="-2.522,-18.228,-1.545", Scale="1,1,1") {
    rotationMode=BMM_Additive
    rotationSpace=BCS_BoneSpace
    scaleMode=BMM_Additive
    scaleSpace=BCS_BoneSpace
    translationMode=BMM_Additive
    translationSpace=BCS_BoneSpace
    alphaBoolBlend {blendOption=Linear}
    boneToModify {boneName=tie_knot_mstr}
    rotation=<rot -2.522 -18.228 -1.545>
   }
   n32 = Component To Local(ComponentPose=@n31)
   n33 = Get bIsCrouching()
   n34 = "Constraint - tie_knot_mstr (PPPPPPP)"!(
    ComponentPose=Local To Component(LocalPose=Blend Poses by bool(BlendPose_0=@n32, BlendPose_1=@n21, bActiveValue=@n33) {blendType=Linear})
    Alpha=1
    ConstraintWeights_0=0.1221
    ConstraintWeights_1:float=0.7487
    ConstraintWeights_2:float=0.0749
    ConstraintWeights_3:float=0.012
    ConstraintWeights_4:float=0.017
    ConstraintWeights_5:float=0.0179
    ConstraintWeights_6:float=0.0072
   ) -> (Pose) {
    alphaBoolBlend {blendOption=Linear}
    boneToModify {boneName=tie_knot_mstr}
    constraintSetup [
     {targetBone {boneName=spine_04}}
     {targetBone {boneName=neck_01}}
     {targetBone {boneName=neck_02}}
     {targetBone {boneName=head}}
     {targetBone {boneName=clavicle_l}}
     {targetBone {boneName=clavicle_r}}
     {targetBone {boneName=tie_knot_02_dyn}}
    ]
    constraintWeights [1,1,1,1,1,1,1]
   }
   n35 = "Constraint - tie_knot_01_dyn (PPPPPPP)"!(
    ComponentPose=@n34
    Alpha=1
    ConstraintWeights_0=0.1221
    ConstraintWeights_1:float=0.7487
    ConstraintWeights_2:float=0.0749
    ConstraintWeights_3:float=0.012
    ConstraintWeights_4:float=0.017
    ConstraintWeights_5:float=0.0179
    ConstraintWeights_6:float=0.0072
   ) -> (Pose) {
    alphaBoolBlend {blendOption=Linear}
    boneToModify {boneName=tie_knot_01_dyn}
    constraintSetup [
     {targetBone {boneName=spine_04}}
     {targetBone {boneName=neck_01}}
     {targetBone {boneName=neck_02}}
     {targetBone {boneName=head}}
     {targetBone {boneName=clavicle_l}}
     {targetBone {boneName=clavicle_r}}
     {targetBone {boneName=tie_knot_02_dyn}}
    ]
    constraintWeights [1,1,1,1,1,1,1]
   }
   n36 = "Constraint - tie_knot_02_dyn (PPPPPPPP)"!(
    ComponentPose=@n35
    Alpha=1
    ConstraintWeights_0=0.021
    ConstraintWeights_1:float=0.1464
    ConstraintWeights_2:float=0.2097
    ConstraintWeights_3:float=0.49
    ConstraintWeights_4:float=0.0592
    ConstraintWeights_5:float=0.0072
    ConstraintWeights_6:float=0.0347
    ConstraintWeights_7:float=0.0319
   ) -> (Pose) {
    alphaBoolBlend {blendOption=Linear}
    boneToModify {boneName=tie_knot_02_dyn}
    constraintSetup [
     {targetBone {boneName=spine_03}}
     {targetBone {boneName=spine_04}}
     {targetBone {boneName=spine_05}}
     {targetBone {boneName=neck_01}}
     {targetBone {boneName=neck_02}}
     {targetBone {boneName=head}}
     {targetBone {boneName=clavicle_l}}
     {targetBone {boneName=clavicle_r}}
    ]
    constraintWeights [1,1,1,1,1,1,1,1]
   }
   n37 = "Constraint - tie_01_dyn (PPPPPPPP)"!(
    ComponentPose=@n36
    Alpha=1
    ConstraintWeights_0=0.0719
    ConstraintWeights_1:float=0.1832
    ConstraintWeights_2:float=0.3235
    ConstraintWeights_3:float=0.206
    ConstraintWeights_4:float=0.0637
    ConstraintWeights_5:float=0.0444
    ConstraintWeights_6:float=0.0675
    ConstraintWeights_7:float=0.0398
   ) -> (Pose) {
    alphaBoolBlend {blendOption=Linear}
    boneToModify {boneName=tie_01_dyn}
    constraintSetup [
     {targetBone {boneName=spine_02}}
     {targetBone {boneName=spine_03}}
     {targetBone {boneName=spine_04}}
     {targetBone {boneName=spine_05}}
     {targetBone {boneName=neck_01}}
     {targetBone {boneName=clavicle_l}}
     {targetBone {boneName=clavicle_r}}
     {targetBone {boneName=deform_pec_l}}
    ]
    constraintWeights [1,1,1,1,1,1,1,1]
   }
   n38 = "Constraint - tie_02_dyn (PPPPPPP)"!(
    ComponentPose=@n37
    Alpha=1
    ConstraintWeights_0=0.0334
    ConstraintWeights_1:float=0.0807
    ConstraintWeights_2:float=0.244
    ConstraintWeights_3:float=0.4505
    ConstraintWeights_4:float=0.1856
    ConstraintWeights_5:float=0.0057
    ConstraintWeights_6:float=0.0001
   ) -> (Pose) {
    alphaBoolBlend {blendOption=Linear}
    boneToModify {boneName=tie_02_dyn}
    constraintSetup [
     {targetBone {boneName=pelvis}}
     {targetBone {boneName=spine_01}}
     {targetBone {boneName=spine_02}}
     {targetBone {boneName=spine_03}}
     {targetBone {boneName=spine_04}}
     {targetBone {boneName=spine_05}}
     {targetBone {boneName=clavicle_r}}
    ]
    constraintWeights [1,1,1,1,1,1,1]
   }
   n39 = "Constraint - tie_03_dyn (PPPPP)"!(
    ComponentPose=@n38
    Alpha=1
    ConstraintWeights_0=0.2673
    ConstraintWeights_1:float=0.2147
    ConstraintWeights_2:float=0.2952
    ConstraintWeights_3:float=0.1946
    ConstraintWeights_4:float=0.0281
   ) -> (Pose) {
    alphaBoolBlend {blendOption=Linear}
    boneToModify {boneName=tie_03_dyn}
    constraintSetup [
     {targetBone {boneName=pelvis}}
     {targetBone {boneName=spine_01}}
     {targetBone {boneName=spine_02}}
     {targetBone {boneName=spine_03}}
     {targetBone {boneName=spine_04}}
    ]
    constraintWeights [1,1,1,1,1]
   }
   n40 = Component To Local(ComponentPose=@n39)
   n41 = Get bRigidBodyEnabled()
   n42 = "Transform (Modify) Bone - Bone: dyn_jacket_fr_l_01_MSTR"(
    ComponentPose=Local To Component(LocalPose=Blend Poses by bool(BlendPose_0=@n7, BlendPose_1=@n40, bActiveValue=@n41) {blendType=Linear})
    Alpha=1
    Translation="4.212,0.776,2.113"
    Rotation="25,10,0"
    Scale="1,1,1"
   ) {
    rotationMode=BMM_Additive
    rotationSpace=BCS_BoneSpace
    scaleMode=BMM_Additive
    scaleSpace=BCS_BoneSpace
    translationMode=BMM_Additive
    translationSpace=BCS_BoneSpace
    alphaBoolBlend {blendOption=Linear}
    alphaScaleBiasClamp {bInterpResult=true}
    boneToModify {boneName=dyn_jacket_fr_l_01_MSTR}
    rotation {pitch=25, yaw=10}
    translation=<4.212 0.776 2.113>
   }
   n43 = "Transform (Modify) Bone - Bone: dyn_jacket_l_01_MSTR"(ComponentPose=@n42, Alpha=1, Translation="0,0,0", Rotation="20,0,0", Scale="1,1,1") {
    rotationMode=BMM_Additive
    rotationSpace=BCS_BoneSpace
    scaleMode=BMM_Additive
    scaleSpace=BCS_BoneSpace
    translationMode=BMM_Additive
    translationSpace=BCS_BoneSpace
    alphaBoolBlend {blendOption=Linear}
    alphaScaleBiasClamp {bInterpResult=true}
    boneToModify {boneName=dyn_jacket_l_01_MSTR}
    rotation {pitch=20}
   }
   n44 = "Transform (Modify) Bone - Bone: dyn_jacket_bkout_l_01_MSTR"(ComponentPose=@n43, Alpha=1, Translation="0,0,0", Rotation="10,0,0", Scale="1,1,1") {
    rotationMode=BMM_Additive
    rotationSpace=BCS_BoneSpace
    scaleMode=BMM_Additive
    scaleSpace=BCS_BoneSpace
    translationMode=BMM_Additive
    translationSpace=BCS_BoneSpace
    alphaBoolBlend {blendOption=Linear}
    alphaScaleBiasClamp {bInterpResult=true}
    boneToModify {boneName=dyn_jacket_bkout_l_01_MSTR}
    rotation {pitch=10}
   }
   n45 = "Transform (Modify) Bone - Bone: dyn_jacket_bk_l_01_MSTR"(ComponentPose=@n44, Alpha=1, Translation="0,0,0", Rotation="-5,0,0", Scale="1,1,1") {
    rotationMode=BMM_Additive
    rotationSpace=BCS_BoneSpace
    scaleMode=BMM_Additive
    scaleSpace=BCS_BoneSpace
    translationMode=BMM_Additive
    translationSpace=BCS_BoneSpace
    alphaBoolBlend {blendOption=Linear}
    alphaScaleBiasClamp {bInterpResult=true}
    boneToModify {boneName=dyn_jacket_bk_l_01_MSTR}
    rotation {pitch=-5}
   }
   n46 = "Transform (Modify) Bone - Bone: dyn_jacket_bk_cr_l_01_MSTR"(ComponentPose=@n45, Alpha=1, Translation="0,0,0", Rotation="0,0,0", Scale="1,1,1") {
    rotationMode=BMM_Additive
    rotationSpace=BCS_BoneSpace
    scaleMode=BMM_Additive
    scaleSpace=BCS_BoneSpace
    translationMode=BMM_Additive
    translationSpace=BCS_BoneSpace
    alphaBoolBlend {blendOption=Linear}
    alphaScaleBiasClamp {bInterpResult=true}
    boneToModify {boneName=dyn_jacket_bk_cr_l_01_MSTR}
   }
   n47 = "Transform (Modify) Bone - Bone: dyn_jacket_fr_r_01_MSTR"(ComponentPose=@n46, Alpha=1, Translation="0,-5,0", Rotation="40,0,0", Scale="1,1,1") {
    rotationMode=BMM_Additive
    rotationSpace=BCS_BoneSpace
    scaleMode=BMM_Additive
    scaleSpace=BCS_BoneSpace
    translationMode=BMM_Additive
    translationSpace=BCS_BoneSpace
    alphaBoolBlend {blendOption=Linear}
    alphaScaleBiasClamp {bInterpResult=true}
    boneToModify {boneName=dyn_jacket_fr_r_01_MSTR}
    rotation {pitch=40}
    translation {y=-5}
   }
   n48 = "Transform (Modify) Bone - Bone: dyn_jacket_r_01_MSTR"(ComponentPose=@n47, Alpha=1, Translation="0,0,0", Rotation="20,0,0", Scale="1,1,1") {
    rotationMode=BMM_Additive
    rotationSpace=BCS_BoneSpace
    scaleMode=BMM_Additive
    scaleSpace=BCS_BoneSpace
    translationMode=BMM_Additive
    translationSpace=BCS_BoneSpace
    alphaBoolBlend {blendOption=Linear}
    alphaScaleBiasClamp {bInterpResult=true}
    boneToModify {boneName=dyn_jacket_r_01_MSTR}
    rotation {pitch=20}
   }
   n49 = "Transform (Modify) Bone - Bone: dyn_jacket_bkout_r_01_MSTR"(ComponentPose=@n48, Alpha=1, Translation="0,0,0", Rotation="10,0,0.068", Scale="1,1,1") {
    rotationMode=BMM_Additive
    rotationSpace=BCS_BoneSpace
    scaleMode=BMM_Additive
    scaleSpace=BCS_BoneSpace
    translationMode=BMM_Additive
    translationSpace=BCS_BoneSpace
    alphaBoolBlend {blendOption=Linear}
    alphaScaleBiasClamp {bInterpResult=true}
    boneToModify {boneName=dyn_jacket_bkout_r_01_MSTR}
    rotation {pitch=10, roll=0.068}
   }
   n50 = "Transform (Modify) Bone - Bone: dyn_jacket_bk_r_01_MSTR"(ComponentPose=@n49, Alpha=1, Translation="0,0,0", Rotation="-5,0,0", Scale="1,1,1") {
    rotationMode=BMM_Additive
    rotationSpace=BCS_BoneSpace
    scaleMode=BMM_Additive
    scaleSpace=BCS_BoneSpace
    translationMode=BMM_Additive
    translationSpace=BCS_BoneSpace
    alphaBoolBlend {blendOption=Linear}
    alphaScaleBiasClamp {bInterpResult=true}
    boneToModify {boneName=dyn_jacket_bk_r_01_MSTR}
    rotation {pitch=-5}
   }
   n51 = "Transform (Modify) Bone - Bone: dyn_jacket_bk_cr_r_01_MSTR"(ComponentPose=@n50, Alpha=1, Translation="0,0,0", Rotation="0,0,0", Scale="1,1,1") {
    rotationMode=BMM_Additive
    rotationSpace=BCS_BoneSpace
    scaleMode=BMM_Additive
    scaleSpace=BCS_BoneSpace
    translationMode=BMM_Additive
    translationSpace=BCS_BoneSpace
    alphaBoolBlend {blendOption=Linear}
    alphaScaleBiasClamp {bInterpResult=true}
    boneToModify {boneName=dyn_jacket_bk_cr_r_01_MSTR}
   }
   n52 = "Transform (Modify) Bone - Bone: dyn_jacket_fr_r_01_MSTR"(ComponentPose=@n51, Alpha=1, Translation="0,0,5", Rotation="0,0,0", Scale="1,1,1") {
    rotationMode=BMM_Additive
    rotationSpace=BCS_BoneSpace
    scaleMode=BMM_Additive
    scaleSpace=BCS_BoneSpace
    translationMode=BMM_Additive
    translationSpace=BCS_BoneSpace
    alphaBoolBlend {blendOption=Linear}
    alphaScaleBiasClamp {bInterpResult=true}
    boneToModify {boneName=dyn_jacket_fr_r_01_MSTR}
    translation {z=5}
   }
   n53 = Get Skydiving_Alpha_Reverse() -> (Skydiving_Alpha_Reverse:double)
   n54 = "Constraint - dyn_jacket_bk_cr_l_01 (R)"(ComponentPose=@n52, Alpha=@n53, ConstraintWeights_0=1) {
    alphaBoolBlend {blendOption=Linear}
    boneToModify {boneName=dyn_jacket_bk_cr_l_01}
    constraintSetup [{transformType=Rotation, targetBone {boneName=dyn_jacket_bk_cr_r_01}}]
    constraintWeights [1]
   }
   n55 = "Constraint - dyn_jacket_bk_cr_l_02 (R)"(ComponentPose=@n54, Alpha=@n53, ConstraintWeights_0=1) {
    alphaBoolBlend {blendOption=Linear}
    boneToModify {boneName=dyn_jacket_bk_cr_l_02}
    constraintSetup [{transformType=Rotation, targetBone {boneName=dyn_jacket_bk_cr_r_02}}]
    constraintWeights [1]
   }
   n56 = "To Float (Boolean)"(InBool=Get bIsSkyDiving())
   n57 = "Transform (Modify) Bone - Bone: dyn_jacket_bk_cr_r_01_MSTR"(ComponentPose=@n55, Alpha=@n56, Translation="0,0,0", Rotation="0,-3,0", Scale="1,1,1") {
    rotationMode=BMM_Additive
    rotationSpace=BCS_BoneSpace
    scaleMode=BMM_Additive
    scaleSpace=BCS_BoneSpace
    translationMode=BMM_Additive
    translationSpace=BCS_BoneSpace
    alphaBoolBlend {blendOption=Linear}
    alphaScaleBiasClamp {bInterpResult=true}
    boneToModify {boneName=dyn_jacket_bk_cr_r_01_MSTR}
    rotation {yaw=-3}
   }
   n58 = "Transform (Modify) Bone - Bone: dyn_jacket_bk_cr_l_01_MSTR"(ComponentPose=@n57, Alpha=@n56, Translation="0,0,2", Rotation="0,3,0", Scale="1,1,1") {
    rotationMode=BMM_Additive
    rotationSpace=BCS_BoneSpace
    scaleMode=BMM_Additive
    scaleSpace=BCS_BoneSpace
    translationMode=BMM_Additive
    translationSpace=BCS_BoneSpace
    alphaBoolBlend {blendOption=Linear}
    alphaScaleBiasClamp {bInterpResult=true}
    boneToModify {boneName=dyn_jacket_bk_cr_l_01_MSTR}
    rotation {yaw=3}
    translation {z=2}
   }
   n59 = "Constraint - Btm_Button_MSTR (PPPPPPPP)"!(
    ComponentPose=@n58
    Alpha=1
    ConstraintWeights_0=0.2907
    ConstraintWeights_1:float=0.0944
    ConstraintWeights_2:float=0.1222
    ConstraintWeights_3:float=0.0577
    ConstraintWeights_4:float=0.0205
    ConstraintWeights_5:float=0.1292
    ConstraintWeights_6:float=0.1972
    ConstraintWeights_7:float=0.088
   ) -> (Pose) {
    alphaBoolBlend {blendOption=Linear}
    boneToModify {boneName=Btm_Button_MSTR}
    constraintSetup [
     {targetBone {boneName=pelvis}}
     {targetBone {boneName=spine_01}}
     {targetBone {boneName=spine_02}}
     {targetBone {boneName=spine_03}}
     {targetBone {boneName=dyn_jacket_bkout_r_01}}
     {targetBone {boneName=dyn_jacket_r_01}}
     {targetBone {boneName=dyn_jacket_fr_r_01}}
     {targetBone {boneName=dyn_jacket_fr_r_02}}
    ]
    constraintWeights [1,1,1,1,1,1,1,1]
   }
   n60 = "Constraint - top_button_LookAt (PPPPPPPP)"!(
    ComponentPose=@n59
    Alpha=1
    ConstraintWeights_0=0.3068
    ConstraintWeights_1:float=0.1182
    ConstraintWeights_2:float=0.155
    ConstraintWeights_3:float=0.0861
    ConstraintWeights_4:float=0.014
    ConstraintWeights_5:float=0.0995
    ConstraintWeights_6:float=0.1577
    ConstraintWeights_7:float=0.0625
   ) -> (Pose) {
    alphaBoolBlend {blendOption=Linear}
    boneToModify {boneName=top_button_LookAt}
    constraintSetup [
     {targetBone {boneName=pelvis}}
     {targetBone {boneName=spine_01}}
     {targetBone {boneName=spine_02}}
     {targetBone {boneName=spine_03}}
     {targetBone {boneName=dyn_jacket_bkout_r_01}}
     {targetBone {boneName=dyn_jacket_r_01}}
     {targetBone {boneName=dyn_jacket_fr_r_01}}
     {targetBone {boneName=dyn_jacket_fr_r_02}}
    ]
    constraintWeights [1,1,1,1,1,1,1,1]
   }
   n61 = "Constraint - btm_button_LookAt (PPPPPPPPP)"!(
    ComponentPose=@n60
    Alpha=1
    ConstraintWeights_0=0.2641
    ConstraintWeights_1:float=0.0748
    ConstraintWeights_2:float=0.0972
    ConstraintWeights_3:float=0.0289
    ConstraintWeights_4:float=0.025
    ConstraintWeights_5:float=0.153
    ConstraintWeights_6:float=0.0126
    ConstraintWeights_7:float=0.2234
    ConstraintWeights_8:float=0.121
   ) -> (Pose) {
    alphaBoolBlend {blendOption=Linear}
    boneToModify {boneName=btm_button_LookAt}
    constraintSetup [
     {targetBone {boneName=pelvis}}
     {targetBone {boneName=spine_01}}
     {targetBone {boneName=spine_02}}
     {targetBone {boneName=spine_03}}
     {targetBone {boneName=dyn_jacket_bkout_r_01}}
     {targetBone {boneName=dyn_jacket_r_01}}
     {targetBone {boneName=dyn_jacket_r_02}}
     {targetBone {boneName=dyn_jacket_fr_r_01}}
     {targetBone {boneName=dyn_jacket_fr_r_02}}
    ]
    constraintWeights [1,1,1,1,1,1,1,1,1]
   }
   n62 = Look At - Bone: Btm_Button_01(ComponentPose=@n61, Alpha=1) {
    alphaBoolBlend {blendOption=Linear}
    boneToModify {boneName=Btm_Button_01}
    lookAtLocation {x=0}
    lookAtTarget {boneReference {boneName=btm_button_LookAt}}
    lookAt_Axis {axis=<1 0>}
   }
   Output Pose(Result=Component To Local(ComponentPose=@n62)) {layerGroup=None}
   n63 = Control Rig!(Source=Copy Pose From Mesh() {bUseAttachedParent=true}, Alpha, Crouching:double=0) -> (Pose) {
    alphaBoolBlend {blendOption=Linear}
    controlRigAssetReference {blueprintRigClass=&/M_MED_SlicedBread_ControlRig.M_MED_SlicedBread_ControlRig_C}
   }
   n64 = "Constraint - tie_knot_mstr (PPPPPPP)"(
    ComponentPose=Local To Component(LocalPose=@n63)
    Alpha=0.65
    ConstraintWeights_0=0.1216
    ConstraintWeights_1=0.7508
    ConstraintWeights_2=0.0739
    ConstraintWeights_3=0.0119
    ConstraintWeights_4=0.0169
    ConstraintWeights_5=0.0179
    ConstraintWeights_6=0.0071
   ) {
    alphaBoolBlend {blendOption=Linear}
    boneToModify {boneName=tie_knot_mstr}
    constraintSetup [
     {targetBone {boneName=spine_04}}
     {targetBone {boneName=neck_01}}
     {targetBone {boneName=neck_02}}
     {targetBone {boneName=head}}
     {targetBone {boneName=clavicle_l}}
     {targetBone {boneName=clavicle_r}}
     {targetBone {boneName=tie_knot_02_dyn}}
    ]
    constraintWeights [1,1,1,1,1,1,1]
   }
   Save cached pose 'PreSim'(Pose=Component To Local(ComponentPose=@n64)) {cacheName=PreSim}
   Trail controller - Bone: tie_03_dyn!(ComponentPose, Alpha=1, FakeVelocity:Vector="1,1,1") -> (Pose) {
    bActorSpaceFakeVel=true
    bLimitRotation=true
    bShowTrailLocation=true
    chainLength=4
    debugLifeTime=0.1
    maxDeltaTime=0.0333
    relaxationSpeedScale=4
    alphaScaleBiasClamp {bInterpResult=true}
    baseJoint {boneName=root}
    planarLimits [
     {drivingBone {boneName=root}, planeTransform {location=<0 0 10>, rotation=<rot 0 0 0>, scale=<1 1 1>}}
     {drivingBone {boneName=spine_03}, planeTransform {location=<0 -20 0>, rotation=<rot 0 15 -89.9999>, scale=<1 1 1>}}
    ]
    rotationLimits [
     {limitMax=<0 -20 20>, limitMin=<0 -40 -20>}
     {limitMax=<0 0 20>, limitMin=<0 -40 -20>}
     {limitMax=<0 0 20>, limitMin=<0 -40 -20>}
     {limitMax=<0 0 20>, limitMin=<0 -40 -20>}
    ]
    rotationOffsets [<0 0 0>, <0 0 0>, <0 0 0>, <0 0 0>]
    trailBone {boneName=tie_03_dyn}
    trailRelaxationSpeed {editorCurveData {keys [
       {arriveTangent=-72.0739, interpMode=RCIM_Cubic, leaveTangent=-72.0739, tangentMode=RCTM_User, time=0.0047, value=45}
       <at 0.9634 2 RCIM_Cubic>
      ]}}
   }
   n65 = Get BwdMovement() -> (BwdMovement:double)
   n66 = Trail controller - Bone: dyn_jacket_bk_cr_r_02!(ComponentPose, Alpha=@n65, FakeVelocity:Vector="0,0,0") -> (Pose) {
    bActorSpaceFakeVel=true
    bShowTrailLocation=true
    debugLifeTime=0.1
    maxDeltaTime=0.0333
    relaxationSpeedScale=3
    alphaScaleBiasClamp {bInterpResult=true}
    baseJoint {boneName=root}
    planarLimits [
     {drivingBone {boneName=root}, planeTransform {location=<0 0 10>, rotation=<rot 0 0 0>, scale=<1 1 1>}}
     {drivingBone {boneName=spine_03}, planeTransform {location=<0 -20 0>, rotation=<rot 0 15 -89.9999>, scale=<1 1 1>}}
    ]
    rotationLimits [{limitMax=<180 180 180>}, {limitMax=<180 180 180>}]
    rotationOffsets [<0 0 0>, <0 0 0>]
    trailBone {boneName=dyn_jacket_bk_cr_r_02}
    trailRelaxationSpeed {editorCurveData {keys [{arriveTangent=-72.0739, interpMode=RCIM_Cubic, leaveTangent=-72.0739, tangentMode=RCTM_User, time=0, value=35}, <at 1 5 RCIM_Cubic>]}}
   }
   n67 = Trail controller - Bone: dyn_jacket_bk_cr_l_02!(ComponentPose=@n66, Alpha=@n65, FakeVelocity:Vector="0,0,0") -> (Pose) {
    bActorSpaceFakeVel=true
    bShowTrailLocation=true
    debugLifeTime=0.1
    maxDeltaTime=0.0333
    relaxationSpeedScale=3
    alphaScaleBiasClamp {bInterpResult=true}
    baseJoint {boneName=root}
    planarLimits [
     {drivingBone {boneName=root}, planeTransform {location=<0 0 10>, rotation=<rot 0 0 0>, scale=<1 1 1>}}
     {drivingBone {boneName=spine_03}, planeTransform {location=<0 -20 0>, rotation=<rot 0 15 -89.9999>, scale=<1 1 1>}}
    ]
    rotationLimits [{limitMax=<180 180 180>}, {limitMax=<180 180 180>}]
    rotationOffsets [<0 0 0>, <0 0 0>]
    trailBone {boneName=dyn_jacket_bk_cr_l_02}
    trailRelaxationSpeed {editorCurveData {keys [{arriveTangent=-72.0739, interpMode=RCIM_Cubic, leaveTangent=-72.0739, tangentMode=RCTM_User, time=0, value=35}, <at 1 5 RCIM_Cubic>]}}
   }
   n68 = Get LeftMovement() -> (LeftMovement:double)
   n69 = Trail controller - Bone: dyn_jacket_fr_l_02!(ComponentPose=@n67, Alpha=@n68, FakeVelocity:Vector="0,0,0") -> (Pose) {
    bActorSpaceFakeVel=true
    bShowTrailLocation=true
    debugLifeTime=0.1
    maxDeltaTime=0.0333
    relaxationSpeedScale=3
    alphaScaleBiasClamp {bInterpResult=true}
    baseJoint {boneName=root}
    planarLimits [
     {drivingBone {boneName=root}, planeTransform {location=<0 0 10>, rotation=<rot 0 0 0>, scale=<1 1 1>}}
     {drivingBone {boneName=spine_03}, planeTransform {location=<0 -20 0>, rotation=<rot 0 15 -89.9999>, scale=<1 1 1>}}
    ]
    rotationLimits [{limitMax=<180 180 180>}, {limitMax=<180 180 180>}]
    rotationOffsets [<0 0 0>, <0 0 0>]
    trailBone {boneName=dyn_jacket_fr_l_02}
    trailRelaxationSpeed {editorCurveData {keys [{arriveTangent=-72.0739, interpMode=RCIM_Cubic, leaveTangent=-72.0739, tangentMode=RCTM_User, time=0, value=35}, <at 1 5 RCIM_Cubic>]}}
   }
   n70 = Trail controller - Bone: dyn_jacket_bk_r_02!(ComponentPose=@n69, Alpha=@n68, FakeVelocity:Vector="0,0,0") -> (Pose) {
    bActorSpaceFakeVel=true
    bShowTrailLocation=true
    debugLifeTime=0.1
    maxDeltaTime=0.0333
    relaxationSpeedScale=3
    alphaScaleBiasClamp {bInterpResult=true}
    baseJoint {boneName=root}
    planarLimits [
     {drivingBone {boneName=root}, planeTransform {location=<0 0 10>, rotation=<rot 0 0 0>, scale=<1 1 1>}}
     {drivingBone {boneName=spine_03}, planeTransform {location=<0 -20 0>, rotation=<rot 0 15 -89.9999>, scale=<1 1 1>}}
    ]
    rotationLimits [{limitMax=<180 180 180>}, {limitMax=<180 180 180>}]
    rotationOffsets [<0 0 0>, <0 0 0>]
    trailBone {boneName=dyn_jacket_bk_r_02}
    trailRelaxationSpeed {editorCurveData {keys [{arriveTangent=-72.0739, interpMode=RCIM_Cubic, leaveTangent=-72.0739, tangentMode=RCTM_User, time=0, value=35}, <at 1 5 RCIM_Cubic>]}}
   }
   n71 = Get RightMovement() -> (RightMovement:double)
   n72 = Trail controller - Bone: dyn_jacket_fr_l_02(ComponentPose=@n70, Alpha=@n71, FakeVelocity="0,0,0") {
    bActorSpaceFakeVel=true
    bShowTrailLocation=true
    debugLifeTime=0.1
    maxDeltaTime=0.0333
    relaxationSpeedScale=3
    alphaScaleBiasClamp {bInterpResult=true}
    baseJoint {boneName=root}
    planarLimits [
     {drivingBone {boneName=root}, planeTransform {location=<0 0 10>, rotation=<rot 0 0 0>, scale=<1 1 1>}}
     {drivingBone {boneName=spine_03}, planeTransform {location=<0 -20 0>, rotation=<rot 0 15 -89.9999>, scale=<1 1 1>}}
    ]
    rotationLimits [{limitMax=<180 180 180>}, {limitMax=<180 180 180>}]
    rotationOffsets [<0 0 0>, <0 0 0>]
    trailBone {boneName=dyn_jacket_fr_l_02}
    trailRelaxationSpeed {editorCurveData {keys [{arriveTangent=-72.0739, interpMode=RCIM_Cubic, leaveTangent=-72.0739, tangentMode=RCTM_User, time=0, value=35}, <at 1 5 RCIM_Cubic>]}}
   }
   n73 = Trail controller - Bone: dyn_jacket_bkout_l_02!(ComponentPose=@n72, Alpha=@n71, FakeVelocity:Vector="0,0,0") -> (Pose) {
    bActorSpaceFakeVel=true
    bShowTrailLocation=true
    debugLifeTime=0.1
    maxDeltaTime=0.0333
    relaxationSpeedScale=3
    alphaScaleBiasClamp {bInterpResult=true}
    baseJoint {boneName=root}
    planarLimits [
     {drivingBone {boneName=root}, planeTransform {location=<0 0 10>, rotation=<rot 0 0 0>, scale=<1 1 1>}}
     {drivingBone {boneName=spine_03}, planeTransform {location=<0 -20 0>, rotation=<rot 0 15 -89.9999>, scale=<1 1 1>}}
    ]
    rotationLimits [{limitMax=<180 180 180>}, {limitMax=<180 180 180>}]
    rotationOffsets [<0 0 0>, <0 0 0>]
    trailBone {boneName=dyn_jacket_bkout_l_02}
    trailRelaxationSpeed {editorCurveData {keys [{arriveTangent=-72.0739, interpMode=RCIM_Cubic, leaveTangent=-72.0739, tangentMode=RCTM_User, time=0, value=35}, <at 1 5 RCIM_Cubic>]}}
   }
   n74 = Trail controller - Bone: dyn_jacket_bk_l_02!(ComponentPose=@n73, Alpha=0.3, FakeVelocity:Vector="0,0,0") -> (Pose) {
    bActorSpaceFakeVel=true
    bLimitRotation=true
    bShowTrailLocation=true
    debugLifeTime=0.1
    maxDeltaTime=0.0333
    relaxationSpeedScale=3
    alphaScaleBiasClamp {bInterpResult=true}
    baseJoint {boneName=pelvis}
    planarLimits [
     {drivingBone {boneName=root}, planeTransform {location=<0 0 10>, rotation=<rot 0 0 0>, scale=<1 1 1>}}
     {drivingBone {boneName=spine_03}, planeTransform {location=<0 -20 0>, rotation=<rot 0 15 -89.9999>, scale=<1 1 1>}}
    ]
    rotationLimits [{limitMax=<0 50 50>, limitMin=<0 -50 -50>}, {limitMax=<0 50 50>, limitMin=<0 -50 -50>}]
    rotationOffsets [<0 0 0>, <0 0 0>]
    trailBone {boneName=dyn_jacket_bk_l_02}
    trailRelaxationSpeed {editorCurveData {keys [{arriveTangent=-72.0739, interpMode=RCIM_Cubic, leaveTangent=-72.0739, tangentMode=RCTM_User, time=0, value=35}, <at 1 5 RCIM_Cubic>]}}
   }
   Trail controller - Bone: dyn_jacket_bk_r_02(ComponentPose=@n74, Alpha=0.3, FakeVelocity="0,0,0") {
    bActorSpaceFakeVel=true
    bLimitRotation=true
    bShowTrailLocation=true
    debugLifeTime=0.1
    maxDeltaTime=0.0333
    relaxationSpeedScale=3
    alphaScaleBiasClamp {bInterpResult=true}
    baseJoint {boneName=pelvis}
    planarLimits [
     {drivingBone {boneName=root}, planeTransform {location=<0 0 10>, rotation=<rot 0 0 0>, scale=<1 1 1>}}
     {drivingBone {boneName=spine_03}, planeTransform {location=<0 -20 0>, rotation=<rot 0 15 -89.9999>, scale=<1 1 1>}}
    ]
    rotationLimits [{limitMax=<0 50 50>, limitMin=<0 -50 -50>}, {limitMax=<0 50 50>, limitMin=<0 -50 -50>}]
    rotationOffsets [<0 0 0>, <0 0 0>]
    trailBone {boneName=dyn_jacket_bk_r_02}
    trailRelaxationSpeed {editorCurveData {keys [{arriveTangent=-72.0739, interpMode=RCIM_Cubic, leaveTangent=-72.0739, tangentMode=RCTM_User, time=0, value=35}, <at 1 5 RCIM_Cubic>]}}
   }
  ]}
 EventGraph {nodes [
   Event BlueprintUpdateAnimation()
   > n1 = Sequence!(execute) -> (then_0, then_1, then_2:Exec, then_3:Exec, then_4:Exec, then_5:Exec, then_6:Exec, then_7:Exec)
   n2 = Branch(execute=@n1.then_1, Condition=Get bIsCrouching())
   > Set Crouching(Crouching:double=1) -> (Output_Get:double)
   Set Crouching(execute=@n2.else, Crouching=0)
   Branch(execute=@n1.then_2, Condition=Get bIsSkyDiving())
   > Set Skydiving_Alpha_Reverse(Skydiving_Alpha_Reverse:double=0.2) -> (Output_Get:double)
   n3 = Branch(execute=@n1.then_3, Condition=OR Boolean(A=Get bIsJumping(), B=Get bIsFalling()))
   > Set JumpingFalling(JumpingFalling:bool=true) -> (Output_Get:bool)
   Set JumpingFalling(execute=@n3.else, JumpingFalling=false)
   n4 = Branch(execute=@n1.then_4, Condition=AND Boolean(A=Get bIsAccelerating2D(), B=Get bIsCrouching()))
   > Set CrouchWalking(CrouchWalking:double=1) -> (Output_Get:double)
   Set CrouchWalking(execute=@n4.else, CrouchWalking=0)
   n5 = Get bIsSkydiveDiveMode()
   n6 = Branch(execute=@n1.then_5, Condition=AND Boolean(A=Get bIsSkyDiving(), B=NOT Boolean(A=@n5)))
   > Set OnlySkydive(OnlySkydive:bool=true) -> (Output_Get:bool)
   Set OnlySkydive(execute=@n6.else, OnlySkydive=false)
   n7 = GetOwnerSkeletalMeshComponent()
   n8 = BreakRotator(InRot=BreakTransform(InTransform=GetDeltaTransformFromRefPose(self=@n7, BoneName=thigh_r, BaseName=pelvis)).Rotation)
   Set thigh_rot_up_r(execute=@n1.then_6, thigh_rot_up_r:double=MapRangeClamped(Value=@n8.Yaw, InRangeA=-15, InRangeB=-100)) -> (Output_Get:double)
   > Set thigh_rot_inr_r(thigh_rot_inr_r:double=MapRangeClamped(Value=@n8.Pitch, InRangeA=0, InRangeB=-10, OutRangeA=1, OutRangeB=0)) -> (Output_Get:double)
   > Set thigh_rot_otr_r(thigh_rot_otr_r:double=MapRangeClamped(Value=@n8.Pitch, InRangeA=20, InRangeB=50)) -> (Output_Get:double)
   n9 = GetOwnerSkeletalMeshComponent()
   n10 = BreakRotator(InRot=BreakTransform(InTransform=GetDeltaTransformFromRefPose(self=@n9, BoneName=thigh_l, BaseName=pelvis)).Rotation)
   > Set thigh_rot_up_l(thigh_rot_up_l:double=MapRangeClamped(Value=@n10.Yaw, InRangeA=-10, InRangeB=-100)) -> (Output_Get:double)
   > Set thigh_rot_inr_l(thigh_rot_inr_l:double=MapRangeClamped(Value=@n10.Pitch, InRangeA=0, InRangeB=-10, OutRangeA=1, OutRangeB=0)) -> (Output_Get:double)
   > Set thigh_rot_otr_l(thigh_rot_otr_l:double=MapRangeClamped(Value=@n10.Pitch, InRangeA=20, InRangeB=50)) -> (Output_Get:double)
   > Set jumpingAlpha(jumpingAlpha:double="To Float (Boolean)"(InBool=Get bIsJumping())) -> (Output_Get:double)
   n11 = Get bIsMoving2D()
   > Set isNotCrouchRunning(isNotCrouchRunning:double="To Float (Boolean)"(InBool=NOT Boolean(A=@n11))) -> (Output_Get:double)
   > Set ParachuteAlpha(ParachuteAlpha:double="To Float (Boolean)"(InBool=Get bIsParachuteOpened())) -> (Output_Get:double)
   > Set skydiveAlpha(skydiveAlpha:double="To Float (Boolean)"(InBool=Get bIsSkyDiving())) -> (Output_Get:double)
   n12 = PlayerMovement(execute=@n1.then_7, AnimInstance=Self-Reference(), XDotProductCoefficient=1, YDotProductCoefficient=1, ZDotProductCoefficient=1)
   n13 = MapRangeClamped(Value=@n12.YMovement, InRangeA=0, InRangeB=-1)
   n14 = float * float(A="Absolute (Float)"(A=@n12), B=0.5)
   > Set BwdMovement(BwdMovement:double=MapRangeClamped(
     Value=float + float(A=@n14, B=@n13) {is=title K2Node_CommutativeAssociativeBinaryOperator}
     InRangeA=0
     InRangeB=1.1
     OutRangeA=0.4
     OutRangeB=0
    )) -> (Output_Get:double)
   n15 = MapRangeClamped(Value=@n12, InRangeA=-0.2, InRangeB=1)
   n16 = float * float(A="Absolute (Float)"(A=@n12.YMovement), B=0.2)
   > Set RightMovement(RightMovement:double=MapRangeClamped(Value=float + float(A=@n16, B=@n15), InRangeA=0, InRangeB=1.2, OutRangeB=0.5)) -> (Output_Get:double)
   n17 = MapRangeClamped(Value=@n12, InRangeA=0.2, InRangeB=-1, OutRangeB=0.3)
   n18 = float * float(A="Absolute (Float)"(A=@n12.YMovement), B=0.2)
   > Set LeftMovement(LeftMovement:double=MapRangeClamped(Value=float + float(A=@n18, B=@n17), InRangeA=0, InRangeB=1.2, OutRangeB=0.5)) -> (Output_Get:double)
   n19 = Gravity Override State Function DEPRECATED(
    execute=@n1.then_0
    CharacterPartAnimInstance=Self-Reference()
    GameTimeInSeconds=GetGameTimeInSeconds()
    GravityOverrideParams=Get Gravity Override Params()
   )
   > Set GravityOverride(GravityOverride=@n19)
   Get bIsCrouchSprinting()
   Set Skydiving_Alpha_Reverse(Skydiving_Alpha_Reverse=0.45)
  ]}
]
variables [
 BwdMovement {type=number}
 Crouching {type=number}
 CrouchWalking {type=number}
 Gravity Override Params {
  type=GravityOverrideParamsStruct
  default {
   skydive_FloatingDown {gravityOverride=<0 0 -800>, windAmplitude=<100 -300 200>, windOffset=<0 -300 200>}
   skydive_GoingUp {gravityOverride=<0 0 -500>}
  }
 }
 GravityOverride {type=Vector}
 isNotCrouchRunning {type=number}
 jumpingAlpha {type=number}
 JumpingFalling {type=boolean}
 LeftMovement {type=number}
 OnlySkydive {type=boolean}
 ParachuteAlpha {type=number}
 RightMovement {type=number}
 skydiveAlpha {type=number}
 Skydiving_Alpha_Reverse {type=number}
 thigh_rot_inr_l {type=number}
 thigh_rot_inr_r {type=number}
 thigh_rot_otr_l {type=number}
 thigh_rot_otr_r {type=number}
 thigh_rot_up_l {type=number}
 thigh_rot_up_r {type=number}
]
