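# Post-pose AnimBP for the SlicedBread male character's face accessory, adding secondary motion to the dyn_hair bone chain.
# Copies the attached parent mesh's pose, runs five partially blended AnimDynamics bones with springs, custom gravity and spine_05 sphere limits, then applies the accessory's Control Rig.
# The EventGraph sets crouch-walk and pistol, rifle and pickaxe flags from movement and the owner's weapon animation type; the AnimGraph shows no direct reads of them.
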AnimBlueprint /BRCosmetics/Characters/Player/Male/Medium/Bodies/M_MED_SlicedBread/Meshes/Parts/M_MED_SlicedBread_FaceAcc_AnimBP
target_skeleton=&/M_MED_SlicedBread_FaceAcc_Skeleton
graphs [
 AnimGraph {nodes [
   n1 = Anim Dynamics - Bone: dyn_hair_01(ComponentPose=Local To Component(LocalPose=Copy Pose From Mesh() {bUseAttachedParent=true}), Alpha=0.3) {
    angularDampingOverride=0.8
    angularSpringConstant=1500
    bAngularSpring=true
    bOverrideAngularDamping=true
    bUseGravityOverride=true
    bUsePlanarLimit=false
    bUseSphericalLimits=true
    simulationSpace=RootRelative
    alphaBoolBlend {blendOption=Linear}
    boundBone {boneName=dyn_hair_01}
    componentAppliedLinearAccClamp=<0 0 0>
    componentLinearAccScale {z=5}
    componentLinearVelScale {z=5}
    gravityOverride {z=-250}
    physicsBodyDefinitions [{
      boundBone {boneName=dyn_hair_01}
      boxExtents=<5 10 10>
      constraintSetup {angularLimitsMax=<45 0 45>, angularLimitsMin=<-45 -45 -45>, angularTarget=<0 0 0>}
      localJointOffset=<3 0 -3>
     }]
    relativeSpaceBone {boneName=pelvis}
    sphericalLimits [
     {limitRadius=50, drivingBone {boneName=spine_05}, sphereLocalOffset=<-38 0 -10>}
     {limitRadius=50, drivingBone {boneName=spine_05}, sphereLocalOffset=<-38 0 15>}
    ]
   }
   n2 = Anim Dynamics - Bone: dyn_hair_02(ComponentPose=@n1, Alpha=0.4) {
    angularDampingOverride=0.999
    angularSpringConstant=1500
    bAngularSpring=true
    bOverrideAngularDamping=true
    bUseGravityOverride=true
    bUsePlanarLimit=false
    bUseSphericalLimits=true
    simulationSpace=RootRelative
    alphaBoolBlend {blendOption=Linear}
    boundBone {boneName=dyn_hair_02}
    componentAppliedLinearAccClamp=<0 0 0>
    componentLinearAccScale {z=5}
    componentLinearVelScale {z=5}
    gravityOverride {z=-250}
    physicsBodyDefinitions [{
      boundBone {boneName=dyn_hair_02}
      boxExtents=<5 10 10>
      constraintSetup {angularLimitsMax=<45 0 45>, angularLimitsMin=<-45 -45 -45>, angularTarget=<0 0 0>}
      localJointOffset=<3 0 -3>
     }]
    relativeSpaceBone {boneName=pelvis}
    sphericalLimits [
     {limitRadius=50, drivingBone {boneName=spine_05}, sphereLocalOffset=<-38 0 -10>}
     {limitRadius=50, drivingBone {boneName=spine_05}, sphereLocalOffset=<-38 0 15>}
    ]
   }
   n3 = Anim Dynamics - Bone: dyn_hair_03(ComponentPose=@n2, Alpha=0.55, ExternalForce="0,0,0") {
    angularDampingOverride=1
    angularSpringConstant=1500
    bAngularSpring=true
    bOverrideAngularDamping=true
    bUseGravityOverride=true
    bUsePlanarLimit=false
    bUseSphericalLimits=true
    simulationSpace=RootRelative
    alphaBoolBlend {blendOption=Linear}
    boundBone {boneName=dyn_hair_03}
    componentAppliedLinearAccClamp=<0 0 0>
    componentLinearAccScale {z=5}
    componentLinearVelScale {z=5}
    gravityOverride {z=-250}
    physicsBodyDefinitions [{
      boundBone {boneName=dyn_hair_03}
      boxExtents=<5 10 10>
      constraintSetup {angularLimitsMax=<45 0 45>, angularLimitsMin=<-45 -45 -45>, angularTarget=<0 0 0>}
      localJointOffset=<3 0 -3>
     }]
    relativeSpaceBone {boneName=pelvis}
    sphericalLimits [
     {limitRadius=50, drivingBone {boneName=spine_05}, sphereLocalOffset=<-38 0 -10>}
     {limitRadius=50, drivingBone {boneName=spine_05}, sphereLocalOffset=<-38 0 15>}
    ]
   }
   n4 = Anim Dynamics - Bone: dyn_hair_04(ComponentPose=@n3, Alpha=0.6, ExternalForce="0,0,0") {
    angularDampingOverride=0.888
    angularSpringConstant=1500
    bAngularSpring=true
    bOverrideAngularDamping=true
    bUseGravityOverride=true
    bUsePlanarLimit=false
    bUseSphericalLimits=true
    simulationSpace=RootRelative
    alphaBoolBlend {blendOption=Linear}
    boundBone {boneName=dyn_hair_04}
    componentAppliedLinearAccClamp=<0 0 0>
    componentLinearAccScale {z=5}
    componentLinearVelScale {z=5}
    gravityOverride {z=-250}
    physicsBodyDefinitions [{
      boundBone {boneName=dyn_hair_04}
      boxExtents=<5 10 10>
      constraintSetup {angularLimitsMax=<45 0 45>, angularLimitsMin=<-45 -45 -45>, angularTarget=<0 0 0>}
      localJointOffset=<3 0 -3>
     }]
    relativeSpaceBone {boneName=pelvis}
    sphericalLimits [
     {limitRadius=50, drivingBone {boneName=spine_05}, sphereLocalOffset=<-38 0 -10>}
     {limitRadius=52, drivingBone {boneName=spine_05}, sphereLocalOffset=<-40 0 15>}
    ]
   }
   n5 = Anim Dynamics - Bone: dyn_hair_05(ComponentPose=@n4, Alpha=0.55, ExternalForce="0,0,0") {
    angularDampingOverride=0.9999
    angularSpringConstant=5000
    bAngularSpring=true
    bOverrideAngularDamping=true
    bUseGravityOverride=true
    bUsePlanarLimit=false
    bUseSphericalLimits=true
    simulationSpace=RootRelative
    alphaBoolBlend {blendOption=Linear}
    boundBone {boneName=dyn_hair_05}
    componentAppliedLinearAccClamp=<0 0 0>
    componentLinearAccScale {z=5}
    componentLinearVelScale {z=5}
    gravityOverride {z=-250}
    physicsBodyDefinitions [{
      boundBone {boneName=dyn_hair_05}
      boxExtents=<5 10 10>
      constraintSetup {angularLimitsMax=<45 0 45>, angularLimitsMin=<-45 -45 -25>, angularTarget=<0 0 0>}
      localJointOffset=<3 0 -3>
     }]
    relativeSpaceBone {boneName=pelvis}
    sphericalLimits [
     {limitRadius=50, drivingBone {boneName=spine_05}, sphereLocalOffset=<-38 0 -10>}
     {limitRadius=50, drivingBone {boneName=spine_05}, sphereLocalOffset=<-37 0 15>}
    ]
   }
   n6 = Control Rig(Source=Component To Local(ComponentPose=@n5)) {
    alphaBoolBlend {blendOption=Linear}
    controlRigAssetReference {blueprintRigClass=&/M_MED_SlicedBread_FaceAcc_CTRL_Rig.M_MED_SlicedBread_FaceAcc_CTRL_Rig_C}
   }
   Output Pose(Result=@n6) {layerGroup=None}
  ]}
 EventGraph {nodes [
   Event BlueprintUpdateAnimation()
   > n1 = Sequence()
   > n2 = Branch(Condition=AND Boolean(A=Get bIsAccelerating2D(), B=Get bIsCrouching()))
   > Set CrouchWalking(CrouchWalking:double=1) -> (Output_Get:double)
   Set CrouchWalking(execute=@n2.else, CrouchWalking=0)
   n3 = Cast To FortPlayerAnimInstance(execute=@n1.then_1, Object=GetAnimInstance(self=GetOwnerSkeletalMeshComponent()))
   > Set isAimingPistol(isAimingPistol:double=0) -> (Output_Get:double)
   > Set isPickaxe(isPickaxe:double=0) -> (Output_Get:double)
   > Set isAimingRifle(isAimingRifle:double=0) -> (Output_Get:double)
   > Set isPickaxeBoo(isPickaxeBoo:bool=false) -> (Output_Get:bool)
   > n4 = Switch on EFortWeaponCoreAnimation!(execute, Selection:EFortWeaponCoreAnimation Enum=Get WeaponCoreAnim(self=@n3)) -> (
    Melee:Exec
    Pistol:Exec
    Shotgun:Exec
    PaperBlueprint:Exec
    Rifle:Exec
    MeleeOneHand:Exec
    MachinePistol:Exec
    RocketLauncher:Exec
    GrenadeLauncher:Exec
    GoingCommando:Exec
    AssaultRifle:Exec
    TacticalShotgun:Exec
    SniperRifle:Exec
    TrapPlacement:Exec
    ShoulderLauncher:Exec
    AbilityDecoTool:Exec
    Crossbow:Exec
    C4:Exec
    RemoteControl:Exec
    DualWield:Exec
    AR_BullPup:Exec
    AR_ForwardGrip:Exec
    MedPackPaddles:Exec
    SMG_P90:Exec
    AR_DrumGun:Exec
    Consumable_Small:Exec
    Consumable_Large:Exec
    Balloon:Exec
    MountedTurret:Exec
    CreativeTool:Exec
    ExplosiveBow:Exec
    AshtonIndigo:Exec
    AshtonChicago:Exec
    MeleeDualWield:Exec
    Unarmed:Exec
   )
   Branch(execute=[@n4.AssaultRifle, @n4.Crossbow, @n4.GrenadeLauncher, @n4.Rifle, @n4.RocketLauncher, @n4.Shotgun, @n4.SniperRifle, @n4.TacticalShotgun])
   > Set isAimingRifle(isAimingRifle=1)
   Branch(execute=[@n4.MachinePistol, @n4.Pistol])
   > Set isAimingPistol(isAimingPistol=1)
   Branch(execute=@n4.Melee)
   > Set isPickaxe(isPickaxe=1)
   > Set isPickaxeBoo(isPickaxeBoo=true)
  ]}
]
variables [
 CrouchWalking {type=number}
 isAimingPistol {type=number}
 isAimingRifle {type=number}
 isPickaxe {type=number}
 isPickaxeBoo {type=boolean}
]
